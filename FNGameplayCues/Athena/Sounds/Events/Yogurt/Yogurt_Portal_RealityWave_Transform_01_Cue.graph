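# Sound cue for the reality wave transform sound of the Yogurt event portal in Athena.
# A single wave player feeds the output directly, with no randomization, mixing or modulation nodes.
# It uses the gameplay events sound class and sets a very large max distance.
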SoundCue /FNGameplayCues/Athena/Sounds/Events/Yogurt/Yogurt_Portal_RealityWave_Transform_01_Cue
root=WavePlayer_0
nodes [{id=WavePlayer_0, properties {soundWaveAssetPtr=&/Yogurt_Portal_RealityWave_Transform_01}}]
properties {maxDistance=2097152, soundClassObject=&Fort_Events_Gameplay}
